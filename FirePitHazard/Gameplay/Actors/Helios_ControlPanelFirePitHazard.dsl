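; Fire pit hazard control panel: an interactable, destructible button that triggers fire traps.
; On authority, BeginPlay destroys it if disabled, disables it without activation data, and otherwise clamps the cooldown.
; Interaction waits a pre-activation delay, fires each entry's pattern through its hazard manager, then resets after cooldown.
; RepNotifies drive button glow, beeping and press sounds, plus a playspace event broadcast on press; death fades the beep.
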(event EventOnDeathPlayEffects
  (params (Damage :type Float%20%28single-precision%29) (DamageTags :type Gameplay%20Tag%20Container%20Structure%20%28by%20ref%29) (Momentum :type Vector :default "0, 0, 0") (HitInfo :type Hit%20Result%20Structure%20%28by%20ref%29) (InstigatedBy :type Fort%20Pawn%20Object%20Reference) (DamageCauser :type Actor%20Object%20Reference) (EffectContext :type Gameplay%20Effect%20Context%20Handle%20Structure))
  (bind _audio (Variables|Default|GetAudio))
  ;[1 1 1] On Death Stop Beeping Audio
  (Utilities|IsValid _audio
    (:Is%20Valid
      (if (Utilities|IsValid _audio :!class /Script/Engine.KismetSystemLibrary)
        (Audio|Components|Audio|FadeOut _audio 0.2 0.0 "Linear" :!class /Script/Engine.AudioComponent)))))

(event EventBeginPlay
  ;[1 1 1] Init
  (Networking|SwitchHasAuthority
    (:Authority
      (if (ScalableFloat|GetValueasBool (Variables|Default|GetSF_Enabled))
        (Audio|Components|Audio|FadeIn (Variables|Default|GetAudio) 0.15 1.0 0.0 "Linear" :!class /Script/Engine.AudioComponent)
        (if (>= (Utilities|Array|Length (Variables|Default|GetActivationData)) 1)
          (Variables|Default|SetCooldownDuration (Math|Float|Clamp(Float) (Variables|Default|GetCooldownDuration) 3.5 99999.0))
          (EventDispatchers|BindEventtoOnDied (delegate Custom|OnDied_Event))
          (else
            (Variables|Default|SetNoDataDisable true)))
        (else
          (Actor|DestroyActor))))))

(event Custom|OnDied_Event
  (params (DamagedActor :type Actor%20Object%20Reference) (Damage :type Float%20%28single-precision%29) (InstigatedBy :type Controller%20Object%20Reference) (DamageCauser :type Actor%20Object%20Reference) (HitLocation :type Vector) (FHitComponent :type Primitive%20Component%20Object%20Reference) (BoneName :type Name) (Momentum :type Vector))
  ;[1 1 1] Init
  (Audio|Components|Audio|FadeOut (Variables|Default|GetAudio) 0.15 0.0 "Linear" :!class /Script/Engine.AudioComponent))

(event Interaction|EventBlueprintOnInteract
  (params (InteractingPawn :type Fort%20Pawn%20Object%20Reference) (InteractionBeingAttempted :type EInteractionBeingAttempted%20Enum :default "FirstInteraction"))
  ;[1 1 1] Interact
  (Variables|Default|SetButtonInteracted true)
  (Utilities|FlowControl|Delay (Variables|Default|GetPreActivationDelay))
  (Utilities|FlowControl|Sequence
    (:then_0
      (for _array_element (Variables|Default|GetActivationData)
        (Class|HeliosFirePitHazardManager|FireAllatPatternIndex (Utilities|Struct|BreakFirePitHazardControlPanelStruct _array_element) (.PatternIndexToFire_6_01E7F07349389119F032CB86205CB988 (Utilities|Struct|BreakFirePitHazardControlPanelStruct _array_element))))
      (Utilities|FlowControl|Delay (Variables|Default|GetCooldownDuration))
      (Variables|Default|SetButtonInteracted false))))

(fn BlueprintCanInteract
  (params (InteractingPawn :type Fort%20Pawn%20Object%20Reference) (InteractionBeingAttempted :type EInteractionBeingAttempted%20Enum :default "FirstInteraction") (InteractionType :type TInteractionType%20Enum :default "IT_NoInteraction"))
  (return (!= (Variables|Default|GetButtonInteracted) (!= (Variables|Default|GetNoDataDisable) true))))

(fn BlueprintGetInteractionString
  (params (InteractingPawn :type Fort%20Pawn%20Object%20Reference) (InteractionBeingAttempted :type EInteractionBeingAttempted%20Enum :default "FirstInteraction"))
  (return "NSLOCTEXT(\"[C50EBC3273171A962FC7BD3C5E929351]\", \"9C508AA845CC09E682E7D88CD7D388FB\", \"Trigger Fire Traps\")"))

(fn OnRep_ButtonInteracted
  (if (Variables|Default|GetButtonInteracted)
    (bind _returnvalue (Rendering|Material|CreateDynamicMaterialInstance (Variables|Default|Getbuttonmesh) 0 "/Game/Athena/Asteria/Maps/Test/Dev/Chris/Actors/FirePits/M_ButtonGlowTest.M_ButtonGlowTest" "None" :!class /Script/Engine.PrimitiveComponent))
    (Rendering|Material|SetScalarParameterValue _returnvalue "pulse" 0.0 :!class /Script/Engine.MaterialInstanceDynamic)
    (Audio|Components|Audio|FadeOut (Variables|Default|GetAudio) 0.1 0.0 "Linear" :!class /Script/Engine.AudioComponent)
    (bind _returnvalue_2 (GameplayEventRouter|GetEventRouter "/Script/FortniteGame.FortPlayspace" self))
    (GameplayEventRouter|BroadcastEvent _returnvalue_2 (Utilities|Struct|MakeFirePitHazardEventStruct true) 0)
    (bind _output_get (Variables|Default|SetAlarmSoundRef 0))
    (Audio|Components|Audio|Play _output_get 0.0)
    (bind _returnvalue_3 (Audio|SpawnSoundAttached :Sound "/FirePitHazard/Assets/Audio/FirePitHazard_Button_Press_Cue.FirePitHazard_Button_Press_Cue" :AttachToComponent (Variables|Default|Getbuttonmesh) :AttachPointName "None" :Location (Transformation|GetActorLocation self) :LocationType "KeepWorldPosition" :bStopWhenAttachedToDestroyed false :VolumeMultiplier 1.0 :PitchMultiplier 1.0 :StartTime 0.0 :bAutoDestroy true))
    (Audio|Components|Audio|Play _returnvalue_3 0.0)
    (else
      (bind _returnvalue_4 (Rendering|Material|CreateDynamicMaterialInstance (Variables|Default|Getbuttonmesh) 0 "/Game/Athena/Asteria/Maps/Test/Dev/Chris/Actors/FirePits/M_ButtonGlowTest.M_ButtonGlowTest" "None" :!class /Script/Engine.PrimitiveComponent))
      (Rendering|Material|SetScalarParameterValue _returnvalue_4 "pulse" 1.0 :!class /Script/Engine.MaterialInstanceDynamic)
      (Audio|PlaySoundatLocation "/FirePitHazard/Assets/Audio/FirePitHazard_Button_Activate_Cue.FirePitHazard_Button_Activate_Cue" (Transformation|GetWorldLocation (Variables|Default|GetAudio)))
      (Audio|Components|Audio|FadeIn (Variables|Default|GetAudio) 0.15 1.0 0.0 "Linear" :!class /Script/Engine.AudioComponent))))

(fn OnRep_NoDataDisable
  (if (Variables|Default|GetNoDataDisable)
    (bind _returnvalue (Rendering|Material|CreateDynamicMaterialInstance (Variables|Default|Getbuttonmesh) 0 "/Game/Athena/Asteria/Maps/Test/Dev/Chris/Actors/FirePits/M_ButtonGlowTest.M_ButtonGlowTest" "None" :!class /Script/Engine.PrimitiveComponent))
    (Rendering|Material|SetScalarParameterValue _returnvalue "pulse" 0.0 :!class /Script/Engine.MaterialInstanceDynamic)))

(fn ConstructionScript)
