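; Fire pit hazard actor that erupts as a timed geyser, launching and damaging players and vehicles and launching ridden actors.
; A manager or player trigger fires it, optionally after warm-up; the server sets IsActive and damages blocking props through async targeting.
; IsActive's RepNotify toggles cone collision and burst effects.
; Overlap applies impulses, multicast for client-authoritative vehicles; leaving grants fall-damage immunity, and hotfix scalable floats drive tuning.
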(event Custom|SetupMaterialParams
  (bind _s_fx_geysercone (Variables|Default|GetS_FX_GeyserCone))
  ;[0.008445 0.4764 1] Set material values
  (Rendering|Material|SetScalarParameterValueonMaterials _s_fx_geysercone "UVScale" (Math|Random|RandomFloatinRange 2.0 4.0))
  (Rendering|Material|SetScalarParameterValueonMaterials _s_fx_geysercone "SlantAmount" (Math|Random|RandomFloatinRange -5e+02 5e+02))
  (Rendering|Material|SetScalarParameterValueonMaterials _s_fx_geysercone "WPO_Amount" (Math|Random|RandomFloatinRange 0.0 5e+02))
  (Rendering|Material|SetScalarParameterValueonMaterials _s_fx_geysercone "WorldTexUVScale" (Math|Random|RandomFloatinRange 0.5 2.0))
  (Rendering|Material|SetScalarParameterValueonMaterials _s_fx_geysercone "Top Edge Contrast" (Math|Random|RandomFloatinRange 1.0 3.0)))

(event EventBeginPlay
  ;[0.05098 1 0.01223] Begin Play
  (if (ScalableFloat|GetValueasBool (Variables|Hotfix|GetSF_Enabled))
    (Variables|Default|SetActiveTime (ScalableFloat|GetValueatLevel (Variables|Hotfix|GetSF_FirePitActiveDuration)))
    (Networking|CanExecuteCosmeticEvents
      (:True
        (CallFunction|SetupMaterialParams)
        (Audio|Components|Audio|FadeIn (Variables|Default|GetAudio) 1.0 1.0 0.0 "Linear" :!class /Script/Engine.AudioComponent)))
    (else
      (Actor|DestroyActor))))

(event Custom|Fire
  ;[1 0 0.9359] fire pits from manager or player button
  (Networking|SwitchHasAuthority
    (:Authority
      (if (not (Variables|Default|GetIsActive))
        (Variables|Default|SetIsActive true)
        (CallFunction|DestroyBuildingProps)
        (bind _returnvalue (Utilities|Time|SetTimerbyEvent (delegate Custom|ResetGeyserWait) (Variables|Default|GetActiveTime)))
        (Variables|Default|SetActiveTimerHandle _returnvalue)))))

(event OnComponentBeginOverlap(S_FX_GeyserCone)
  (params (OverlappedComponent :type Primitive%20Component%20Object%20Reference) (OtherActor :type Actor%20Object%20Reference) (OtherComp :type Primitive%20Component%20Object%20Reference) (OtherBodyIndex :type Integer) (bFromSweep :type Boolean) (SweepResult :type Hit%20Result%20Structure%20%28by%20ref%29))
  ;[1 0.3378 0] Launch players and vehicles if they overlap the geyser while it's active. Handle projectiles and toys/C4 separately.
  (bind _asfort_player_pawn (Utilities|Casting|CastToFortPlayerPawn OtherActor)
    (:then
      (CallFunction|LaunchPlayerorMount _asfort_player_pawn))
    (:CastFailed
      (bind _asfort_vehicle_interface (Utilities|Casting|CastToFortVehicleInterface OtherActor))
      (Networking|SwitchHasAuthority
        (:Authority
          (CallFunction|LaunchVehicle _asfort_vehicle_interface))) ; Vehicles are only collided with on the server! \nSo this will only be called in the server anyway
    ))
  (CallFunction|TriggerVentLoopingGameplayCue OtherActor))

(event Custom|ResetGeyserWait
  ;[1 0 0.9359] fire pits from manager or player button
  (Variables|Default|SetIsActive false))

(event Custom|WarmUp
  ;[1 0 0.9359] fire pits from manager or player button
  (GameplayCue|ExecuteGameplayCue self (Variables|Default|GetGCN_WarmUp))
  (bind _returnvalue (Utilities|Time|SetTimerbyEvent (EventDispatchers|CreateEvent :!event fIRE) (ScalableFloat|GetValueatLevel (Variables|Hotfix|GetSF_FirePitWarmUpDuration))))
  (Variables|Default|SetWarmUpTimerHandle _returnvalue))

(event Custom|ResetAndForceFire
  ;[1 0 0.9359] fire pits from manager or player button
  ;;[1 1 1] called when a player fires all pits manually - stops currently active ones and resets them then fires
  (Utilities|Time|ClearandInvalidateTimerbyHandle (Variables|Default|GetActiveTimerHandle))
  (Utilities|Time|ClearandInvalidateTimerbyHandle (Variables|Default|GetWarmUpTimerHandle))
  (Variables|Default|SetIsActive false)
  (CallFunction|Fire))

(event OnComponentEndOverlap(S_FX_GeyserCone)
  (params (OverlappedComponent :type Primitive%20Component%20Object%20Reference) (OtherActor :type Actor%20Object%20Reference) (OtherComp :type Primitive%20Component%20Object%20Reference) (OtherBodyIndex :type Integer))
  ;[0.1134 0 1] Set the launched targets to be immune to fall damage.
  (bind _asfort_pawn (Utilities|Casting|CastToFortPawn OtherActor)
    (:then
      (Ability|SendGameplayEventtoActor _asfort_pawn "(TagName=\"Abilities.Generic.Vent.Exit\")")
      (Variables|/Script/FortniteGame.FortPawn|SetbIgnoreNextFallingDamage true _asfort_pawn)
      (Projectile|RemoveGameplayCue%28Looping%29Local _asfort_pawn (Variables|Default|GetVentCue) (Ability|GameplayCue|MakeGameplayCueParameters :NormalizedMagnitude 0.0 :RawMagnitude 0.0 :EffectContext "()" :AggregatedSourceTags "(GameplayTags=)" :AggregatedTargetTags "(GameplayTags=)" :Location "0,0,0" :Normal "0,0,0" :GameplayEffectLevel 1 :AbilityLevel 1 :bReplicateLocationWhenUsingMinimalRepProxy false)))
    (:CastFailed
      (bind _asfort_vehicle_interface (Utilities|Casting|CastToFortVehicleInterface OtherActor))
      (Vehicle|SetIgnoreAllFallingDamage _asfort_vehicle_interface false)
      (Vehicle|SetIgnoreNextFallingDamage _asfort_vehicle_interface true))))

(event Custom|MulticastLaunchVehicle
  (params (VehicleToLaunch :type Fort%20Vehicle%20Interface%20Interface) (LinearVelocity :type Vector%20%28by%20ref%29))
  ;[1 0.4724 0] Multicast version of Impulse vehicle, we need to call this in cliients as vehicles are client authoritative
  (Vehicle|ImpulseVehicle VehicleToLaunch LinearVelocity (Math|Vector|VectorZero) false false true "None"))

(event Custom|DestroyBuildingProps
  ;[1 1 1] Destroys Building Props in the way of the pit
  (bind (_asynctaskproxy _targetinghandle) (Class|AsyncActionPerformTargeting|PerformTargetingAsyncAction 0 "/FirePitHazard/Gameplay/TP_FirePitHazard_DetectDestructibles.TP_FirePitHazard_DetectDestructibles" true)
    (:Targeted
      (Ability|Tasks|ApplyGameplayEffectToTargetingHandleAsync (Ability|GetFortAbilitySystemComponent self) _targetinghandle "/Game/Athena/Items/Consumables/ShockwaveGrenade/GE_ShockGrenade_Damage.GE_ShockGrenade_Damage_C" 0 1))) ; Note: We could use this request for players and vehicles\nas well, if perf shows up issues in this front
)

(fn GetCollisionStaticMesh
  (return "0, 0, 0" "/Game/Athena/Environments/Island/Runes/Mesh/SM_CorruptedVent_A.SM_CorruptedVent_A"))

(fn GetMeshComponents
  (return 0))

(fn IgnoreWorldGridScaleOverride
  (return false))

(fn LaunchPlayerOrMount
  (params (OverlappingPlayerPawn :type Fort%20Player%20Pawn%20Object%20Reference))
  (if (Vehicle|IsinVehicle OverlappingPlayerPawn)
    _
    (elif (FortMovement|IsRidingonActor OverlappingPlayerPawn)
      ;[1 1 1] Launch the actor we're riding
      (FortPhysicsSimulationLibrary|AddPhysicsImpulse (Pawn|GetMovementBaseActor OverlappingPlayerPawn) (* (Transformation|GetActorUpVector) (ScalableFloat|GetValueatLevel (Variables|Hotfix|GetSF_PhysicsActorLaunchVelocity))) true "None" false false self "(MaxSimSizeAllowed=Large,bRejectEmbeddedInLandscape=False)")
      (else
        ;[1 1 1] Launch player
        (Ability|SendGameplayEventtoActor (Variables|GetOverlappingPlayerPawn) "(TagName=\"Abilities.Generic.Vent.Enter\")" (Utilities|Struct|MakeGameplayEventData :Target self :ContextHandle "()" :InstigatorTags "(GameplayTags=)" :TargetTags "(GameplayTags=)" :EventMagnitude (Variables|Default|GetUpwardLaunchVelocity) :TargetData "()"))
        (Environment|ApplyGlobalEnvironmentDamagetoActor (Variables|GetOverlappingPlayerPawn) (ScalableFloat|GetValueatLevel (Variables|Hotfix|GetSF_PlayerDamage)) (GameplayTags|GetOwnedGameplayTags :!class /Script/GameplayTags.GameplayTagAssetInterface))
        (Character|LaunchCharacter (Variables|GetOverlappingPlayerPawn) (* (Transformation|GetActorUpVector) (ScalableFloat|GetValueatLevel (Variables|Hotfix|GetSF_LaunchVelocity))) false true)))))

(fn LaunchVehicle
  (params (VehicleToLaunch :type Fort%20Vehicle%20Interface%20Interface))
  (LocalVariables|SetVehicleActor (Vehicle|GetActor VehicleToLaunch))
  (CallFunction|MulticastLaunchVehicle (Variables|GetVehicleToLaunch) (+ (* (Transformation|GetActorUpVector) (ScalableFloat|GetValueatLevel (Variables|Hotfix|GetSF_VehicleLaunchVelocity))) (Math|Vector|SelectVector (Vehicle|GetVehicleForwardVector (Variables|GetVehicleToLaunch)) (* (Vehicle|GetVehicleForwardVector (Variables|GetVehicleToLaunch)) (* (ScalableFloat|GetValueatLevel (Variables|Hotfix|GetSF_VehicleExtraFwdLaunchVelocity)) (Math|Float|SelectFloat 1.0 -1.0 (> (Vehicle|GetForwardSpeed (Variables|GetVehicleToLaunch)) 0)))) (< (Math|Float|Absolute(Float) (Vehicle|GetForwardSpeed (Variables|GetVehicleToLaunch))) 5.0)))) ; Vehicles are client authoritative so we need to call\nthis in server and client for the impulse to happen
  (Vehicle|SetIgnoreAllFallingDamage (Variables|GetVehicleToLaunch) true)
  (Environment|ApplyGlobalEnvironmentDamagetoActor (LocalVariables|GetVehicleActor) (ScalableFloat|GetValueatLevel (Variables|Hotfix|GetSF_VehicleDamage)) (GameplayTags|GetOwnedGameplayTags :!class /Script/GameplayTags.GameplayTagAssetInterface)))

(fn OnRep_IsActive
  (bind _geyserburstref (Variables|Default|GetGeyserBurstRef))
  (if (Variables|Default|GetIsActive)
    (Collision|SetCollisionEnabled (Variables|Default|GetS_FX_GeyserCone) "QueryAndPhysics")
    (bind _returnvalue (Effects|Components|ParticleSystem|SpawnFXAtLocation "/FirePitHazard/Assets/VFX/NS_FirePitHazard_FireBall.NS_FirePitHazard_FireBall" (Transformation|GetWorldLocation (Variables|Default|GetFortPhysicsOverlap))))
    (Variables|Default|SetGeyserBurstRef _returnvalue)
    (GameplayCue|ExecuteGameplayCueOnActor(Burst) self (Variables|Default|GetLaunchGC) (Ability|GameplayCue|MakeGameplayCueParameters 0.0))
    (else
      (Collision|SetCollisionEnabled (Variables|Default|GetS_FX_GeyserCone) "NoCollision")
      (Utilities|IsValid _geyserburstref
        (:Is%20Valid
          (Components|Activation|Deactivate _geyserburstref))))))

(fn TriggerVentLoopingGameplayCue
  (params (Target :type Actor%20Object%20Reference))
  (bind (_returnvalue_x _returnvalue_y _returnvalue_z) (Math|Vector|BreakVector (Transformation|GetActorLocation Target)))
  (bind (_returnvalue_x_2 _returnvalue_y_2 _returnvalue_z_2) (Math|Vector|BreakVector (Transformation|GetActorLocation)))
  (Networking|CanExecuteCosmeticEvents
    (:True
      (Projectile|AddGameplayCue%28Looping%29Local Target (Variables|Default|GetVentCue) (Ability|GameplayCue|MakeGameplayCueParameters :NormalizedMagnitude (Math|Conversions|ToFloat(Boolean) (<= _returnvalue_z (+ _returnvalue_z_2 8343.651367))) :RawMagnitude 0.0 :EffectContext "()" :AggregatedSourceTags "(GameplayTags=)" :AggregatedTargetTags "(GameplayTags=)" :Location "0,0,0" :Normal "0,0,0" :GameplayEffectLevel 1 :AbilityLevel 1 :bReplicateLocationWhenUsingMinimalRepProxy false))
      (ForceFeedback|SpawnForceFeedbackAttached "/Game/Weapons/ForceFeedback/Explosions/FF_Generic_ExplosionLarge_Near.FF_Generic_ExplosionLarge_Near" (Variables|/Script/Engine.Actor|GetRootComponent (Variables|GetTarget))))))

(fn ConstructionScript
  (|CallParentFunction))
