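; Server-side manager actor that sequences fire pit hazards through configured patterns, destroying itself when hotfix-disabled.
; Each pattern, optionally shuffled, queues bursts of pits for warm-up, with randomized timer delays between bursts and patterns and optional looping.
; A repeating timer processes one queued pit per tick to spread work across frames.
; FireAllAtPatternIndex force-fires a whole pattern, clearing queued warm-ups and pausing normal playback until the queue drains.
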(event EventBeginPlay
  (Networking|SwitchHasAuthority
    (:Authority
      (if (ScalableFloat|GetValueasBool (Variables|Hotfix|GetSF_Enabled))
        (CallFunction|TriggerAllPatterns)
        (else
          (Actor|DestroyActor))))))

(event Custom|TriggerAllPatterns
  ;[1 1 1] Will trigger all patterns from scratch
  (if (Variables|Default|GetEnabled)
    (CallFunction|ResetManager)
    (CallFunction|PrintLog "INVTEXT(\"Started Firing All Pits from Scratch\")")
    (CallFunction|AdvancePatternSequence)))

(event Custom|AdvancePatternSequence
  (bind _allfirepitpatterns (Variables|Default|GetAllFirePitPatterns))
  ;[1 1 1] Will Try to play the next pattern in the sequence or restart all the patterns form 0 if we're finished
  (bind _result (Math|Integer|IncrementInt (Variables|Default|GetCurrentPatternDataIndex))
    (:%20%20
      (if (Utilities|Array|IsValidIndex _allfirepitpatterns _result)
        (bind _output_get (Variables|Default|SetCurrentPatternData (Utilities|Array|Get(aref) _allfirepitpatterns _result)))
        (CallFunction|InitializeFirePattern _output_get)
        (CallFunction|TryAdvanceBurst)
        (elif (Variables|Default|GetLoopWhenCompleted)
          (CallFunction|TriggerAllPatterns))))))

(event Custom|TryAdvanceBurst
  (bind (_currentpatterndata_patterntype_19_cc09dfd4438cba56c24e00872410ed91 _currentpatterndata_allfirepitsinpattern_18_27d3b29341c94313967c10a3b2537734 _currentpatterndata_numperburst_13_b036009140fed3484aff569200297eb6 _currentpatterndata_firedelay_5_12f4ca544e1bb6c0a9fd6398e9872bbe _currentpatterndata_randomfiredelaymax_17_f02d5cf945b5720ccd75129047a2a45a) (Utilities|Struct|BreakFirePitsHazardDataStruct (Variables|GetCurrentPatternData_PatternType_19_CC09DFD4438CBA56C24E00872410ED91)))
  (bind _returnvalue (Math|Random|RandomFloatinRange 0.0 _currentpatterndata_randomfiredelaymax_17_f02d5cf945b5720ccd75129047a2a45a))
  (bind _returnvalue_2 (Math|Float|Clamp(Float) _currentpatterndata_firedelay_5_12f4ca544e1bb6c0a9fd6398e9872bbe 0.5 9999.0))
  (bind _returnvalue_3 (+ _returnvalue_2 _returnvalue))
  ;[1 1 1] Based of where the Current Fire Pit Index is, it will try to play the next burst for the current pattern. If we finish the pattern, it will go to the next one, otherwise it will keep playing the bursts for the current one.
  (bind _finishedpattern (CallFunction|TryTriggerCurrentFirePitBurst))
  (if _finishedpattern
    (bind _returnvalue_4 (Utilities|Time|SetTimerbyEvent (EventDispatchers|CreateEvent :!event Custom|AdvancePatternSequence) _returnvalue_3 false true))
    (Variables|Default|SetAdvancePatternSequenceTimerHandle _returnvalue_4)
    (CallFunction|PrintLog (Utilities|Text|FormatText "INVTEXT(\"Pattern {Index} Finished. Waiting {WaitTime} seconds to advance to next pattern\")" (Variables|Default|GetCurrentPatternDataIndex) _returnvalue_3))
    (else
      (bind _returnvalue_5 (Utilities|Time|SetTimerbyEvent (EventDispatchers|CreateEvent :!event Custom|TryAdvanceBurst) _returnvalue_3 false true))
      (Variables|Default|SetAdvanceBurstTimerHandle _returnvalue_5)
      (CallFunction|PrintLog (Utilities|Text|FormatText "INVTEXT(\"Pattern {Index} unfinished. Waiting {WaitTime} seconds to advance to next burst\")" (Variables|Default|GetCurrentPatternDataIndex) _returnvalue_3)))))

(event Custom|ProcessAllFirePitQueues
  ;[1 1 1] Processes all the queued pits for firing, this will disperse fire calls through different frames. This will cover both the ones we added to the queue for normal Warm Up flow and the ones we want to force fire and skip warm up flow
  (if (> (Utilities|Array|Length (Variables|Default|GetPendingPitsToForceFire)) 0)
    (CallFunction|ProcessQueue (Variables|Default|GetPendingPitsToForceFire) true)
    (elif (> (Utilities|Array|Length (Variables|Default|GetPendingPitsToWarmUp)) 0)
      (CallFunction|ProcessQueue (Variables|Default|GetPendingPitsToWarmUp))))
  (if (not (CallFunction|AreThereAnyPitstoProcess))
    (CallFunction|StopProcessingFirePitQueues)))

(event Custom|FireAllAtPatternIndex
  (params (PatternIndex :type Integer))
  (bind _allfirepitpatterns (Variables|Default|GetAllFirePitPatterns))
  (bind _output (Utilities|Array|Get(acopy) _allfirepitpatterns PatternIndex))
  (bind (_patterntype_19_cc09dfd4438cba56c24e00872410ed91 _allfirepitsinpattern_18_27d3b29341c94313967c10a3b2537734) (Utilities|Struct|BreakFirePitsHazardDataStruct _output))
  ;[1 1 1] Force Fire all the Fire Pits at the same time (scattered in frames) for a certain Pattern Index
  (if (Utilities|Array|IsValidIndex _allfirepitpatterns PatternIndex)
    (Utilities|Array|AppendArray (Variables|Default|GetPendingPitsToForceFire) _allfirepitsinpattern_18_27d3b29341c94313967c10a3b2537734)
    (CallFunction|PrintLog (Utilities|Text|FormatText "INVTEXT(\"Added {NumPitsAppended} fire pits at index {Index}  to the Queue for Forced Fire\")" (Utilities|Array|Length _allfirepitsinpattern_18_27d3b29341c94313967c10a3b2537734) PatternIndex))
    (Utilities|Array|Clear (Variables|Default|GetPendingPitsToWarmUp)) ; Clear any queued warm ups so \nthey don't overlap the forced fires
    (CallFunction|StartProccessingFirePitQueues)
    (CallFunction|SetTimersPaused true) ; Pause Normal Flow
  ))

(event Custom|StartProccessingFirePitQueues
  ;[1 1 1] Starts the "tick" timer to process queues if not active already
  (if (not (Utilities|Time|IsTimerActivebyHandle (Variables|Default|GetFirePitManagerTickHandle)))
    (bind _returnvalue (Utilities|Time|SetTimerbyEvent (EventDispatchers|CreateEvent :!event Custom|ProcessAllFirePitQueues) (ScalableFloat|GetValueatLevel (Variables|Hotfix|GetHF_PitProcessingTickInterval)) true true))
    (Variables|Default|SetFirePitManagerTickHandle _returnvalue)
    (CallFunction|PrintLog "INVTEXT(\"Started Fire Pit Processing Timer\")")))

(event Custom|StopProcessingFirePitQueues
  ;[1 1 1] Ends The tick timer and will resume normal pattern processing after (paused if forced pattern triggering was summoned)
  (Utilities|Time|ClearandInvalidateTimerbyHandle (Variables|Default|GetFirePitManagerTickHandle))
  (CallFunction|PrintLog "INVTEXT(\"Cleared Fire Pit Processing Timer\")")
  (if (Utilities|Time|IsTimerPausedbyHandle (Variables|Default|GetAdvancePatternSequenceTimerHandle))
    (CallFunction|SetTimersPaused false)
    (CallFunction|PrintLog "INVTEXT(\"Finished Processing, resuiming normal pattern playback\")")))

(fn AreThereAnyPitsToProcess
  (return (> (+ (Utilities|Array|Length (Variables|Default|GetPendingPitsToForceFire)) (Utilities|Array|Length (Variables|Default|GetPendingPitsToWarmUp))) 0)))

(fn InitializeFirePattern
  (params (FirePattern :type Fire%20Pits%20Hazard%20Data%20Struct%20Structure))
  (bind (_patterntype_19_cc09dfd4438cba56c24e00872410ed91 _allfirepitsinpattern_18_27d3b29341c94313967c10a3b2537734) (Utilities|Struct|BreakFirePitsHazardDataStruct FirePattern))
  (bind _currentfirepitsinpattern (Variables|Default|GetCurrentFirePitsInPattern))
  (Variables|Default|SetCurrentFirePitIndex 0)
  (Utilities|Array|Clear _currentfirepitsinpattern)
  (Utilities|Array|AppendArray _currentfirepitsinpattern _allfirepitsinpattern_18_27d3b29341c94313967c10a3b2537734)
  (switch Utilities|FlowControl|Switch|SwitchonFirePitHazardPatterns _patterntype_19_cc09dfd4438cba56c24e00872410ed91
    (:NewEnumerator0
      _)
    (:NewEnumerator3
      (Utilities|Array|Shuffle (Variables|Default|GetCurrentFirePitsInPattern))))
  (return))

(fn PrintLog
  (params (Text :type Text))
  ;[1 1 1] This is a glorified Verbose log
  (if (ScalableFloat|GetValueasBool (Variables|Hotfix|GetHF_EnableVeryVerboseLogs))
    (if true
      (Fortnite|FortShippingLog self (Utilities|String|ToString(Text) (Utilities|Text|FormatText "INVTEXT(\"LogFirePitHazardManager[{FrameNumber}]: {LogText}\")" (Utilities|GetFrameCount) Text))))))

(fn ProcessQueue
  (params (QueueToProcess :type Array%20of%20Helios%20Fire%20Pit%20Hazard%20Object%20References) (bShouldForceFire :type Boolean))
  (bind _output (Utilities|Array|Get(acopy) QueueToProcess))
  ;[1 1 1] Spreads out Pit Firing in Several Threads
  (if (Utilities|Array|IsValidIndex QueueToProcess)
    (Utilities|IsValid _output
      (:Is%20Valid
        (if (Variables|GetbShouldForceFire)
          (Class|HeliosFirePitHazard|ResetandForceFire _output)
          (else
            (Class|HeliosFirePitHazard|WarmUp _output)))
        (CallFunction|PrintLog (Utilities|Text|FormatText "INVTEXT(\"Processing Fire Pit {FirePit}. ShouldForceFire = {ShuldForceFire}. Then popping it from the queue.\")" _output (Variables|GetbShouldForceFire)))
        (Utilities|Array|RemoveIndex QueueToProcess) ; Pop
      ))))

(fn ResetManager
  (bind _pendingpitstowarmup (Variables|Default|GetPendingPitsToWarmUp))
  (bind _pendingpitstoforcefire (Variables|Default|GetPendingPitsToForceFire))
  (Utilities|Time|ClearandInvalidateTimerbyHandle (Variables|Default|GetFirePitManagerTickHandle))
  (Utilities|Time|ClearandInvalidateTimerbyHandle (Variables|Default|GetAdvancePatternSequenceTimerHandle))
  (Utilities|Time|ClearandInvalidateTimerbyHandle (Variables|Default|GetAdvanceBurstTimerHandle))
  (if (not (Utilities|Array|IsEmpty _pendingpitstowarmup))
    (Utilities|Array|Clear _pendingpitstowarmup))
  (if (not (Utilities|Array|IsEmpty _pendingpitstoforcefire))
    (Utilities|Array|Clear _pendingpitstoforcefire))
  (Variables|Default|SetCurrentPatternDataIndex -1))

(fn SetTimersPaused
  (params (bShouldPause :type Boolean))
  (CallFunction|PrintLog (Utilities|Text|FormatText "INVTEXT(\"Set Timers Pause {bShouldPause}\")" bShouldPause))
  (if bShouldPause
    (Utilities|Time|PauseTimerbyHandle (Variables|Default|GetAdvancePatternSequenceTimerHandle))
    (Utilities|Time|PauseTimerbyHandle (Variables|Default|GetAdvanceBurstTimerHandle))
    (else
      (Utilities|Time|UnpauseTimerbyHandle (Variables|Default|GetAdvancePatternSequenceTimerHandle))
      (Utilities|Time|UnpauseTimerbyHandle (Variables|Default|GetAdvanceBurstTimerHandle)))))

(fn TryTriggerCurrentFirePitBurst
  (bind (_currentpatterndata_patterntype_19_cc09dfd4438cba56c24e00872410ed91 _currentpatterndata_allfirepitsinpattern_18_27d3b29341c94313967c10a3b2537734 _currentpatterndata_numperburst_13_b036009140fed3484aff569200297eb6) (Utilities|Struct|BreakFirePitsHazardDataStruct (Variables|GetCurrentPatternData_PatternType_19_CC09DFD4438CBA56C24E00872410ED91)))
  (bind _currentfirepitsinpattern (Variables|Default|GetCurrentFirePitsInPattern))
  (bind _currentfirepitindex (Variables|Default|GetCurrentFirePitIndex))
  ;[1 1 1] Tries to fire "Num Per Burst" pits, if we have less than that (at least one) it will fire that one and say it's finished
  (if (< (Variables|Default|GetCurrentFirePitIndex) (Utilities|Array|Length (Variables|Default|GetCurrentFirePitsInPattern)))
    (LocalVariables|SetFirstBurstIndex _currentfirepitindex) ; Cache these because they will be evaluated every loop 
    (LocalVariables|SetLastBurstIndex (+ _currentfirepitindex (- _currentpatterndata_numperburst_13_b036009140fed3484aff569200297eb6 1)))
    (CallFunction|PrintLog (Utilities|Text|FormatText "INVTEXT(\"Processing {PitBurstNum} Pit Burst(s). Starting at Index {FirstIndex} til {LastIndex}\")" _currentpatterndata_numperburst_13_b036009140fed3484aff569200297eb6 (LocalVariables|GetFirstBurstIndex) (LocalVariables|GetLastBurstIndex)))
    (bind _index (Utilities|FlowControl|ForLoopwithBreak (LocalVariables|GetFirstBurstIndex) (LocalVariables|GetLastBurstIndex))
      (:LoopBody
        (bind _output_get (Variables|Default|SetCurrentFirePitIndex _index))
        (if (Utilities|Array|IsValidIndex (Variables|Default|GetCurrentFirePitsInPattern) _index)
          _
          (else
            _))
        (Utilities|IsValid (Utilities|Array|Get(acopy) _currentfirepitsinpattern _output_get)
          (:Is%20Valid
            (Utilities|Array|Add (Variables|Default|GetPendingPitsToWarmUp) (Utilities|Array|Get(acopy) _currentfirepitsinpattern _output_get))
            (CallFunction|PrintLog (Utilities|Text|FormatText "INVTEXT(\"Added all fire pit {FirePit} at index {Index}  to the Process Queue for Warm Up\")" (Utilities|Array|Get(acopy) _currentfirepitsinpattern _output_get) (Variables|Default|GetCurrentFirePitIndex))))))
      (:Completed
        _))
    ;;[0 0.1803 1] Make sure we increment index after loop completes so we don't use the last Pit index from the last pattern; also ensures we will start over when hitting the end
    (Variables|Default|SetCurrentFirePitIndex (+ (Variables|Default|GetCurrentFirePitIndex) 1))
    (CallFunction|StartProccessingFirePitQueues)
    (return (>= (Variables|Default|GetCurrentFirePitIndex) (Utilities|Array|Length (Variables|Default|GetCurrentFirePitsInPattern))))))

(fn ConstructionScript)
